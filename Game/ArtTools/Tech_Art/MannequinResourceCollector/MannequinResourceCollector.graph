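# Editor utility widget that lists materials, skeletal meshes and character parts for the selected Fort Player Mannequin's character, backpack, pickaxe or glider.
# Selection changes, and Tick polling of variant tags and cosmetic IDs, trigger rebuilds of the asset arrays.
# Material overrides and active variant swaps are applied to those arrays, using embedded Python and CosmeticPickerModule.
# Checkboxes and tabs choose which of the widget's own properties its Details Views display.
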EditorUtilityWidgetBlueprint /Game/ArtTools/Tech_Art/MannequinResourceCollector/MannequinResourceCollector
graphs [
 CheckIDChanged {nodes [
   CheckIDChanged() {is=title K2Node_FunctionEntry}
   > n1 = Execute Python Script!(
    execute
    PythonScript="import unreal\r\nimport sys\r\nimport csv\r\n\r\nsettingsClassString = \"/Script/FortniteGame.CosmeticDeveloperSettings\"\r\nsettingsSectionName = \"CosmeticLoadOutOverrides\"\r\nsettingsClass = unreal.load_class(None, settingsClassString) ##get the class of the Developer Settings\r\nsettingsCdo = unreal.get_default_object(settingsClass) ##Get the default object so we can edit its properties\r\noverrides = settingsCdo.get_editor_property(settingsSectionName)\r\n\r\n\r\ncid = overrides.get_editor_property(\"Character\")\r\nbid = overrides.get_editor_property(\"Backpack\")\r\npid = overrides.get_editor_property(\"Pickaxe\")\r\ngid = overrides.get_editor_property(\"Glider\")"
   ) -> (then, ReturnValue, bid:Object Reference, pid:Object Reference, gid:Object Reference) {outputs [bid,pid,gid]}
   > n2 = Branch(Condition="Equal (Object)"(A=@n1.bid, B=Get BackpackID() -> (BackpackID:Athena Backpack Item Definition Object Reference)))
   > n3 = Branch(Condition="Equal (Object)"(A=@n1.pid, B=Get PickaxeID() -> (PickaxeID:Athena Pickaxe Item Definition Object Reference)))
   > n4 = Branch(Condition="Equal (Object)"(A=@n1.gid, B=Get GliderID() -> (GliderID:Athena Glider Item Definition Object Reference)))
   > Return Node!(execute, IDChanged:bool=false)
   Return Node(execute=[@n2.else, @n4.else, @n3.else, @n5.else], IDChanged=true)
   n5 = Branch(Condition="Equal (Object)"(B=Get CharacterID() -> (CharacterID:Athena Character Item Definition Object Reference)))
  ]}
 CleanUp {nodes [
   CleanUp() {is=title K2Node_FunctionEntry}
   > n1 = Sequence!(execute) -> (then_0, then_1, then_2:Exec, then_3:Exec, then_4:Exec)
   > Clear(TargetArray:Array of Material Interface Object References=Get BodyMaterials() -> (BodyMaterials:Array of Material Interface Object References))
   > Clear(TargetArray=Get HeadMaterials() -> (HeadMaterials:Array of Material Interface Object References))
   > Clear(TargetArray=Get FaceAccMaterials() -> (FaceAccMaterials:Array of Material Interface Object References))
   > Clear(TargetArray=Get HatMaterials() -> (HatMaterials:Array of Material Interface Object References))
   > Clear(TargetArray=Get MiscTailMaterials() -> (MiscTailMaterials:Array of Material Interface Object References))
   Clear(
    execute=@n1.then_1
    TargetArray:Array of Skeletal Mesh Object References=Get BodySkelMeshes() -> (BodySkelMeshes:Array of Skeletal Mesh Object References)
   )
   > Clear(TargetArray=Get HeadSkelMeshes() -> (HeadSkelMeshes:Array of Skeletal Mesh Object References))
   > Clear(TargetArray=Get FaceSkelMeshes() -> (FaceSkelMeshes:Array of Skeletal Mesh Object References))
   > Clear(TargetArray=Get HatSkelMeshes() -> (HatSkelMeshes:Array of Skeletal Mesh Object References))
   > Clear(TargetArray=Get MiscTailSkelMeshes() -> (MiscTailSkelMeshes:Array of Skeletal Mesh Object References))
   Clear(
    execute=@n1.then_2
    TargetArray:Array of Custom Character Part Object References=Get BodyCPs() -> (BodyCPs:Array of Custom Character Part Object References)
   )
   > Clear(TargetArray=Get HeadCPs() -> (HeadCPs:Array of Custom Character Part Object References))
   > Clear(TargetArray=Get FaceCPs() -> (FaceCPs:Array of Custom Character Part Object References))
   > Clear(TargetArray=Get HatCPs() -> (HatCPs:Array of Custom Character Part Object References))
   > Clear(TargetArray=Get MiscTailCPs() -> (MiscTailCPs:Array of Custom Character Part Object References))
   Clear!(
    execute=@n1.then_3
    TargetMap:Map of Strings to Material Interface Object References=Get BodyOverlayMaterials() -> (BodyOverlayMaterials:Map of Strings to Material Interface Object References)
   ) -> (then) {is=title K2Node_CallFunction}
   > Clear(TargetMap=Get HeadOverlayMaterials() -> (HeadOverlayMaterials:Map of Strings to Material Interface Object References))
   > Clear(TargetMap=Get FaceAccOverlayMaterials() -> (FaceAccOverlayMaterials:Map of Strings to Material Interface Object References))
   > Clear(TargetMap=Get HatOverlayMaterials() -> (HatOverlayMaterials:Map of Strings to Material Interface Object References))
   > Clear(TargetMap=Get MiscTailOverlayMaterials() -> (MiscTailOverlayMaterials:Map of Strings to Material Interface Object References))
   Clear(
    execute=@n1.then_4
    TargetArray:Array of Material Interface Object References=Get Materials() -> (Materials:Array of Material Interface Object References)
   ) {is=title K2Node_CallArrayFunction}
   > Clear(TargetArray:Array of Skeletal Mesh Object References=Get Meshes() -> (Meshes:Array of Skeletal Mesh Object References))
   > Clear(TargetMap=Get AccOverlayMaterials() -> (AccOverlayMaterials:Map of Strings to Material Interface Object References)) {is=title K2Node_CallFunction}
  ]}
 EventGraph {nodes [
   "On Clicked (Btn_Material)"() {is=title K2Node_ComponentBoundEvent}
   > n1 = OnTabClicked!(execute, self, TabIndex:int=0) -> (then)
   UpdateCosmeticAsset(execute=[@n7.then, @n6.then, @n2.then, @n3.then, @n1.then])
   "On Clicked (Btn_CP)"() {is=title K2Node_ComponentBoundEvent}
   > n2 = OnTabClicked(TabIndex=2)
   "On Clicked (Btn_SkelMesh)"() {is=title K2Node_ComponentBoundEvent}
   > n3 = OnTabClicked(TabIndex=1)
   "On Clicked (Btn_Refresh)"() {is=title K2Node_ComponentBoundEvent}
   > n4 = GetSelectionSet(self=Get LevelEditorSubsystem())
   > OnSelectionChanged!(execute, self, SelectionSet:Typed Element Selection Set Object Reference=@n4) -> (then)
   "On Check State Changed (Checkbox_Backpack)"!() -> (OutputDelegate, then, bIsChecked:bool) {is=title K2Node_ComponentBoundEvent}
   > UpdateCheckboxes!(
    execute
    self
    Checked:Editor Utility Check Box Object Reference=Get Checkbox_Backpack() -> (Checkbox_Backpack:Editor Utility Check Box Object Reference)
   ) -> (then)
   > Set CurrAssetType(CurrAssetType:int=2) -> (Output_Get:int=0)
   > n5 = Set CurrCosmeticType(CurrCosmeticType:CosmeticPicker_CosmeticTypes Enum=NewEnumerator1) -> (Output_Get:CosmeticPicker_CosmeticTypes Enum=NewEnumerator0)
   n6 = Set isCharacter(execute=[@n8.then, @n9.then, @n5.then], isCharacter:bool=false) -> (Output_Get:bool)
   "On Check State Changed (Checkbox_Character)"!() -> (OutputDelegate, then, bIsChecked:bool) {is=title K2Node_ComponentBoundEvent}
   > UpdateCheckboxes(Checked=Get Checkbox_Character() -> (Checkbox_Character:Editor Utility Check Box Object Reference))
   > Set CurrAssetType(CurrAssetType=0) -> (Output_Get=0)
   > Set CurrCosmeticType(CurrCosmeticType=NewEnumerator0) -> (Output_Get=NewEnumerator0)
   > n7 = Set isCharacter(isCharacter=true)
   "On Check State Changed (Checkbox_Glider)"!() -> (OutputDelegate, then, bIsChecked:bool) {is=title K2Node_ComponentBoundEvent}
   > UpdateCheckboxes(Checked=Get Checkbox_Glider() -> (Checkbox_Glider:Editor Utility Check Box Object Reference))
   > Set CurrAssetType(CurrAssetType=3) -> (Output_Get=0)
   > n8 = Set CurrCosmeticType(CurrCosmeticType=NewEnumerator3) -> (Output_Get=NewEnumerator0)
   "On Check State Changed (Checkbox_Pickaxe)"!() -> (OutputDelegate, then, bIsChecked:bool) {is=title K2Node_ComponentBoundEvent}
   > UpdateCheckboxes(Checked=Get Checkbox_Pickaxe() -> (Checkbox_Pickaxe:Editor Utility Check Box Object Reference))
   > Set CurrAssetType(CurrAssetType=1) -> (Output_Get=0)
   > n9 = Set CurrCosmeticType(CurrCosmeticType=NewEnumerator2) -> (Output_Get=NewEnumerator0)
   UpdateCosmeticAsset() {is=title K2Node_CustomEvent}
   > CleanUp() {is=title K2Node_CallFunction}
   > UpdateUI()
   > n10 = GetMannequin!(
    execute
    self
    SelectionSet:Typed Element Selection Set Object Reference=Get CurrSelection Set() -> (CurrSelection Set:Typed Element Selection Set Object Reference)
   ) -> (then, Mannequin:Fort Player Mannequin Object Reference)
   > n11 = Switch on CosmeticPicker_CosmeticTypes!(
    execute
    Selection:CosmeticPicker_CosmeticTypes Enum=Get CurrCosmeticType() -> (CurrCosmeticType:CosmeticPicker_CosmeticTypes Enum=NewEnumerator0)
   ) -> (
    NewEnumerator0:Exec
    NewEnumerator1:Exec
    NewEnumerator2:Exec
    NewEnumerator3:Exec
    NewEnumerator4:Exec
    NewEnumerator5:Exec
    NewEnumerator6:Exec
    NewEnumerator7:Exec
    NewEnumerator8:Exec
    NewEnumerator9:Exec
    NewEnumerator10:Exec
    NewEnumerator11:Exec
    NewEnumerator12:Exec
    NewEnumerator13:Exec
    NewEnumerator14:Exec
    NewEnumerator15:Exec
   ) {is=title K2Node_SwitchEnum}
   > n12 = GetCustomCharacterParts!(execute, self:Fort Player Mannequin Object Reference=@n10) -> (then, ReturnValue:Array of Custom Character Part Object References)
   > n13 = For Each Loop(Array:Array of Custom Character Part Object References=@n12) -> (Array Element:Custom Character Part Object Reference)
   LoadVariants!(execute=@n13.Completed, self, Mannequin:Fort Player Mannequin Object Reference=@n10) -> (then)
   n14 = Switch on EFortCustomPartType!(
    execute=@n13.LoopBody
    Selection:EFortCustomPartType Enum=GetCharacterPartType!(self:Custom Character Part Object Reference=@n13) -> (ReturnValue:EFortCustomPartType Enum=Head)
   ) -> (Head:Exec, Body:Exec, Hat:Exec, Backpack:Exec, MiscOrTail:Exec, Face:Exec, Gameplay:Exec, ExtraPart:Exec, Gameplay2:Exec) {is=title K2Node_SwitchEnum}
   LoadPartItems!(
    execute=@n14.Body
    self
    CharacterPart:Custom Character Part Object Reference=@n13
    Materials:Array of Material Interface Object References=Get BodyMaterials()
    Meshes:Array of Skeletal Mesh Object References=Get BodySkelMeshes()
    CPs:Array of Custom Character Part Object References=Get BodyCPs()
    OverlayMaterials:Map of Strings to Material Interface Object References=Get BodyOverlayMaterials()
   ) -> (then)
   LoadPartItems(
    execute=@n14.Face
    CharacterPart=@n13
    Materials=Get FaceAccMaterials()
    Meshes=Get FaceSkelMeshes()
    CPs=Get FaceCPs()
    OverlayMaterials=Get FaceAccOverlayMaterials()
   )
   LoadPartItems(
    execute=@n14.Hat
    CharacterPart=@n13
    Materials=Get HatMaterials()
    Meshes=Get HatSkelMeshes()
    CPs=Get HatCPs()
    OverlayMaterials=Get HatOverlayMaterials()
   )
   LoadPartItems(
    execute=@n14.Head
    CharacterPart=@n13
    Materials=Get HeadMaterials()
    Meshes=Get HeadSkelMeshes()
    CPs=Get HeadCPs()
    OverlayMaterials=Get HeadOverlayMaterials()
   )
   LoadPartItems(
    execute=@n14.MiscOrTail
    CharacterPart=@n13
    Materials=Get MiscTailMaterials()
    Meshes=Get MiscTailSkelMeshes()
    CPs=Get MiscTailCPs()
    OverlayMaterials=Get MiscTailOverlayMaterials()
   )
   n15 = Get BackpackID()
   Is Valid(exec=@n11.NewEnumerator1, InputObject=@n15)
   > n16 = LoadPartItems(
    CharacterPart="Get (a copy)"(Array:Array of Custom Character Part Object References=GetCharacterParts!(self:Athena Character Part Item Definition Object Reference=@n15) -> (ReturnValue:Array of Custom Character Part Object References)) -> (Output:Custom Character Part Object Reference)
    Materials=Get Materials()
    Meshes=Get Meshes()
    OverlayMaterials=Get AccOverlayMaterials()
   )
   LoadVariants(execute=[@n16.then, @n22.then, @n19.then], Mannequin=@n10)
   n17 = Get PickaxeID()
   Is Valid(exec=@n11.NewEnumerator2, InputObject=@n17)
   n18 = Get WeaponDefinition!(self:Athena Pickaxe Item Definition Object Reference=@n17) -> (WeaponDefinition:Fort Weapon Melee Item Definition Object Reference)
   > Set Meshes(Meshes:Array of Skeletal Mesh Object References=Make Array("[0]":Skeletal Mesh Object Reference=GetWeaponMeshOverride!(self:Fort Weapon Item Definition Object Reference=@n18) -> (ReturnValue:Skeletal Mesh Object Reference)) -> (Array:Array of Skeletal Mesh Object References)) -> (Output_Get:Array of Skeletal Mesh Object References)
   > n19 = LoadPartItems(Materials=Get Materials(), OverlayMaterials=Get AccOverlayMaterials())
   n20 = Get GliderID()
   Is Valid(exec=@n11.NewEnumerator3, InputObject=@n20)
   > n21 = Execute Python Script!(execute, PythonScript="mesh = id.get_editor_property(\"SkeletalMesh\")", id:Athena Glider Item Definition Object Reference=@n20) -> (then, ReturnValue, mesh:Skeletal Mesh Object Reference) {inputs [id], outputs [mesh]}
   > Set Meshes(Meshes=Make Array("[0]"=@n21.mesh))
   > n22 = LoadPartItems(Materials=Get Materials(), OverlayMaterials=Get AccOverlayMaterials())
   n23 = OnSelectionChanged!() -> (OutputDelegate, then, SelectionSet:Typed Element Selection Set Object Reference) {is=title K2Node_CustomEvent}
   > n24 = Set CurrSelection Set(CurrSelection Set:Typed Element Selection Set Object Reference=@n23.SelectionSet) -> (Output_Get:Typed Element Selection Set Object Reference)
   > n25 = Execute Python Script!(
    execute
    PythonScript="import unreal\r\nimport sys\r\nimport csv\r\n\r\nsettingsClassString = \"&CosmeticDeveloperSettings\"\r\nsettingsSectionName = \"CosmeticLoadOutOverrides\"\r\nsettingsClass = unreal.load_class(None, settingsClassString) ##get the class of the Developer Settings\r\nsettingsCdo = unreal.get_default_object(settingsClass) ##Get the default object so we can edit its properties\r\noverrides = settingsCdo.get_editor_property(settingsSectionName)\r\n\r\n\r\ncid = overrides.get_editor_property(\"Character\")\r\nbid = overrides.get_editor_property(\"Backpack\")\r\npid = overrides.get_editor_property(\"Pickaxe\")\r\ngid = overrides.get_editor_property(\"Glider\")"
   ) -> (
    then
    ReturnValue
    bid:Athena Backpack Item Definition Object Reference
    pid:Athena Pickaxe Item Definition Object Reference
    gid:Athena Glider Item Definition Object Reference
   ) {outputs [bid,pid,gid]}
   > Set BackpackID(BackpackID:Athena Backpack Item Definition Object Reference=@n25.bid) -> (Output_Get:Athena Backpack Item Definition Object Reference)
   > Set PickaxeID(PickaxeID:Athena Pickaxe Item Definition Object Reference=@n25.pid) -> (Output_Get:Athena Pickaxe Item Definition Object Reference)
   > Set GliderID(GliderID:Athena Glider Item Definition Object Reference=@n25.gid) -> (Output_Get:Athena Glider Item Definition Object Reference)
   > n26 = GetMannequin(SelectionSet=@n24)
   > Is Valid(InputObject=@n26)
   > Set CharacterID(CharacterID:Athena Character Item Definition Object Reference=GetAthenaCharacterItemDefinition!(self:Fort Player Mannequin Object Reference=@n26) -> (ReturnValue:Athena Character Item Definition Object Reference)) -> (Output_Get:Athena Character Item Definition Object Reference)
   > UpdateCosmeticAsset() {is=title K2Node_CallFunction}
   Event Tick()
   n27 = Get CurrSelection Set()
   > n28 = GetMannequin(SelectionSet=@n27)
   > Is Valid(InputObject=@n28)
   > n29 = Execute Python Script!(
    execute
    PythonScript="variants = man.get_editor_property(\"OverrideVariants\")\r\ntags = []\r\nfor var in variants:\r\n    tag = var.get_editor_property(\"ActiveVariantTag\")\r\n    tags.append(tag)"
    man:Fort Player Mannequin Object Reference=@n28
   ) -> (then, ReturnValue, tags:Array of Gameplay Tag Structures) {inputs [man], outputs [tags]}
   > n30 = Branch(Condition=Identical!(
     ArrayA:Array of Gameplay Tag Structures=@n29.tags
     ArrayB:Array of Gameplay Tag Structures=Get VariantTags() -> (VariantTags:Array of Gameplay Tag Structures)
    ) -> (ReturnValue:bool) {is=title K2Node_CallArrayFunction})
   > n31 = CheckIDChanged!(execute, self) -> (then, IDChanged:bool) {is=title K2Node_CallFunction}
   > n32 = Branch(Condition=@n31)
   OnSelectionChanged(execute=[@n32.then, @n30.else], SelectionSet=@n27) {is=title K2Node_CallFunction}
   > Set VariantTags(VariantTags:Array of Gameplay Tag Structures=@n29.tags) -> (Output_Get:Array of Gameplay Tag Structures)
   Event Construct()
   > n33 = GetSelectionSet(self=Get LevelEditorSubsystem())
   > Bind Event to On Selection Change(self:Typed Element Selection Set Object Reference=@n33, Delegate=@n23) {is=title K2Node_AddDelegate}
   > OnSelectionChanged(SelectionSet=@n33)
   > Set CurrCosmeticType(CurrCosmeticType=NewEnumerator0) -> (Output_Get=NewEnumerator0)
   > Set isCharacter(isCharacter=true)
   > Initialize()
   Event Destruct()
  ]}
 GetMannequin {nodes [
   n1 = GetMannequin!() -> (then, SelectionSet:Typed Element Selection Set Object Reference) {is=title K2Node_FunctionEntry}
   > Is Valid(InputObject=@n1)
   > n2 = GetElementHandles!(
    execute
    ElementList:Typed Element List Structure=GetNormalizedSelection!(
     SelectionSet:Typed Element Selection Set Object Reference=@n1
     NormalizationOptions_bExpandGroups:bool=true
     NormalizationOptions_bFollowAttachment:bool=False
    ) -> (ReturnValue:Typed Element List Structure)
    BaseInterfaceType:Interface Interface
   ) -> (then, ReturnValue:Array of Script Typed Element Handle Structures)
   > Branch(Condition="integer > integer"(A=Length(TargetArray:Array of Script Typed Element Handle Structures=@n2)))
   n3 = "Get (a ref)"(Array:Array of Script Typed Element Handle Structures=@n2, Dimension 1=0) -> (Output:"Script Typed Element Handle Structure (by ref)")
   > n4 = GetObject!(
    execute:Exec
    self:Object Reference=GetElementInterface!(
     self:Typed Element Registry Object Reference=Get Default Typed Element Registry!() -> (ReturnValue:Typed Element Registry Object Reference) {is=title K2Node_CallFunction}
     InElementHandle:"Script Typed Element Handle Structure (by ref)"=@n3
     InBaseInterfaceType:Interface Interface=/Script/TypedElementRuntime.TypedElementObjectInterface
    ) -> (ReturnValue:Object Reference)
    InElementHandle:"Script Typed Element Handle Structure (by ref)"=@n3
   ) -> (then:Exec, ReturnValue:Object Reference) {is=title K2Node_Message}
   > Is Valid(InputObject=@n4)
   > n5 = Cast To FortPlayerMannequin!(execute, Object=@n4) -> (then, CastFailed, AsFort Player Mannequin:Fort Player Mannequin Object Reference) {is=title K2Node_DynamicCast}
   > Return Node!(execute, Mannequin:Fort Player Mannequin Object Reference=@n5)
   n6 = GetAthenaCharacterItemDefinition(self=Get ThisMannequin() -> (ThisMannequin:Fort Player Mannequin Object Reference))
   PrintString(InString=Append(A="[MP] current selection: ", B=GetDisplayName(Object=@n6)))
   PrintString(InString=FortPalyerMannequin CAST FAILED)
  ]}
 GetMaterialsOnSkelMesh {nodes [
   n1 = GetMaterialsOnSkelMesh!() -> (then, SkelMesh:Skeletal Mesh Object Reference, Materials:Array of Material Interface Object References) {is=title K2Node_FunctionEntry}
   n2 = Get GliderID()
   > n3 = Is Valid(InputObject=@n2)
   > n4 = Execute Python Script!(execute, PythonScript="mesh = id.get_editor_property(\"SkeletalMesh\")", id:Athena Glider Item Definition Object Reference=@n2) -> (then, ReturnValue, mesh:Object Reference) {inputs [id], outputs [mesh]}
   n5 = Get PickaxeID()
   n6 = SelectCosmeticObject!(
    execute=[@n4.then, @n3.Is Not Valid]
    self
    Character:Object Reference=@n1
    Pickaxe:Object Reference=GetWeaponMeshOverride(self=Get WeaponDefinition(self=@n5))
    Backpack:Object Reference=@n1
    Glider:Object Reference=@n4.mesh
   ) -> (then, Object:Object Reference)
   > n7 = Cast To SkeletalMesh!(execute, Object=@n6) -> (then, CastFailed, AsSkeletal Mesh:Skeletal Mesh Object Reference) {is=title K2Node_DynamicCast}
   > n8 = For Each Loop(Array:Array of Skeletal Material Structures=Get Materials!(self:Skeletal Mesh Object Reference=@n7) -> (Materials:Array of Skeletal Material Structures)) -> (Array Element:Skeletal Material Structure)
   n9 = Break Skeletal Material!(SkeletalMaterial:"Skeletal Material Structure (by ref)"=@n8) -> (
    MaterialInterface:Material Interface Object Reference
    MaterialSlotName:Name
    UVChannelData:Mesh UVChannel Info Structure
    OverlayMaterialInterface:Material Interface Object Reference
   ) {is=title K2Node_BreakStruct}
   > Branch(Condition=IsValid(Object=@n9))
   > Add(
    TargetArray:Array of Material Interface Object References=Get Materials!() -> (Materials:Array of Material Interface Object References)
    NewItem:"Material Interface Object Reference (by ref)"=@n9
   ) {is=title K2Node_CallArrayFunction}
  ]}
 GetOverlayMaterialsFromCP {nodes [
   n1 = GetOverlayMaterialsFromCP!() -> (then, cp:Object Reference, OverlayMaterials:Map of Strings to Material Interface Object References) {is=title K2Node_FunctionEntry}
   > n2 = Execute Python Script!(
    execute
    PythonScript="overlay_materials = cp.get_editor_property(\"OverlayMaterialData\")\r\nlow = overlay_materials.get_editor_property(\"LowQualityMaterial\")\r\nhigh = overlay_materials.get_editor_property(\"HighQualityMaterial\")"
    cp:Object Reference=@n1
   ) -> (then, ReturnValue, low:"Material Interface Object Reference (by ref)", high:"Material Interface Object Reference (by ref)") {inputs [cp], outputs [low,high]}
   > Is Valid(InputObject=@n2.low)
   > Add!(
    execute
    TargetMap:Map of Strings to Material Interface Object References=@n1.OverlayMaterials
    Key:"String (by ref)"=Low
    Value:"Material Interface Object Reference (by ref)"=@n2.low
   ) -> (then) {is=title K2Node_CallFunction}
   > Is Valid(InputObject=@n2.high)
   > Add(TargetMap=@n1.OverlayMaterials, Key=High, Value=@n2.high)
  ]}
 HideDetailsView {nodes [
   n1 = HideDetailsView!() -> (then, Details View:Details View Object Reference) {is=title K2Node_FunctionEntry}
   > SetObject(self=@n1)
   > SetVisibility(self=@n1, InVisibility=Collapsed)
  ]}
 Initialize {nodes [
   Initialize() {is=title K2Node_FunctionEntry}
   > Clear(TargetArray:Array of Editor Utility Button Object References=Get Buttons() -> (Buttons:Array of Editor Utility Button Object References)) {is=title K2Node_CallArrayFunction}
   > Set Buttons(Buttons:Array of Editor Utility Button Object References=Make Array!(
     "[0]":Editor Utility Button Object Reference=Get Btn_Material() -> (Btn_Material:Editor Utility Button Object Reference)
     "[1]":Editor Utility Button Object Reference=Get Btn_SkelMesh() -> (Btn_SkelMesh:Editor Utility Button Object Reference)
     "[2]":Editor Utility Button Object Reference=Get Btn_CP() -> (Btn_CP:Editor Utility Button Object Reference)
    ) -> (Array:Array of Editor Utility Button Object References)) -> (Output_Get:Array of Editor Utility Button Object References)
   > Clear(TargetArray:Array of Editor Utility Check Box Object References=Get Checkboxes() -> (Checkboxes:Array of Editor Utility Check Box Object References))
   > Set Checkboxes(Checkboxes:Array of Editor Utility Check Box Object References=Make Array!(
     "[0]":Editor Utility Check Box Object Reference=Get Checkbox_Character()
     "[1]":Editor Utility Check Box Object Reference=Get Checkbox_Pickaxe()
     "[2]":Editor Utility Check Box Object Reference=Get Checkbox_Backpack()
     "[3]":Editor Utility Check Box Object Reference=Get Checkbox_Glider()
    ) -> (Array:Array of Editor Utility Check Box Object References)) -> (Output_Get:Array of Editor Utility Check Box Object References)
   > OnTabClicked(TabIndex=0)
  ]}
 LoadPartItems {nodes [
   n1 = LoadPartItems!() -> (
    then
    CharacterPart:Custom Character Part Object Reference
    Materials:Array of Material Interface Object References
    Meshes:Array of Skeletal Mesh Object References
    CPs:Array of Custom Character Part Object References
    OverlayMaterials:Map of Strings to Material Interface Object References
   ) {is=title K2Node_FunctionEntry}
   > Clear(TargetArray:Array of Material Interface Object References=@n1.Materials)
   > Clear(TargetArray:Array of Skeletal Mesh Object References=@n1.Meshes)
   > Clear(TargetArray:Array of Custom Character Part Object References=@n1.CPs)
   > Clear(TargetMap=@n1.OverlayMaterials) {is=title K2Node_CallFunction}
   n2 = Get CurrAssetType() -> (CurrAssetType:int=0)
   > n3 = Set IsPickaxeOrGlider(IsPickaxeOrGlider:bool=OR Boolean(A="Equal (Integer)"(A=@n2, B=3), B="Equal (Integer)"(A=@n2, B=1))) -> (Output_Get:bool)
   > n4 = Branch(Condition=@n3)
   GetMaterialsOnSkelMesh!(
    execute=[@n12.then, @n4.then]
    self
    SkelMesh:Skeletal Mesh Object Reference=Get CurrSkelMesh() -> (CurrSkelMesh:Skeletal Mesh Object Reference)
    Materials:Array of Material Interface Object References=Get Materials()
   ) -> (then) {is=title K2Node_CallFunction}
   n5 = Get CharacterPart() -> (CharacterPart:Custom Character Part Object Reference)
   > n6 = SelectCosmeticObject(Character=@n5, Pickaxe=Get WeaponDefinition(self=Get PickaxeID()), Backpack=@n5, Glider=Get GliderID())
   > n7 = Execute Python Script!(
    execute
    PythonScript="if cosmetic_type == 1:\r\n    material_overrides = cp.get_editor_property(\"WeaponMaterialOverrides\")\r\nelse:\r\n    material_overrides = cp.get_editor_property(\"MaterialOverrides\")\r\nret = materials\r\nfor mat in material_overrides:\r\n    override_index = mat.get_editor_property(\"MaterialOverrideIndex\")\r\n    ret[override_index] = mat.get_editor_property(\"OverrideMaterial\")"
    cp=@n6
    cosmetic_type:int=Get CurrAssetType() -> (CurrAssetType=0)
    materials:Array of Material Interface Object References=Get Materials()
   ) -> (then, ReturnValue, ret:Array of Material Interface Object References) {inputs [cp, cosmetic_type, materials], outputs [ret]}
   > n8 = For Each Loop(Array:Array of Material Interface Object References=@n7.ret) -> (Array Element:Material Interface Object Reference)
   > Set Array Elem!(
    execute
    TargetArray:Array of Material Interface Object References=Get Materials()
    Index:int=@n8.Array Index
    Item:"Material Interface Object Reference (by ref)"=@n8
    bSizeToFit:bool=false
   ) -> (then)
   n9 = Get CharacterPart()
   n10 = SelectCosmeticObject(execute=@n8.Completed, Character=@n9, Pickaxe=Get WeaponDefinition(self=Get PickaxeID()), Backpack=@n9, Glider=Get GliderID())
   > GetOverlayMaterialsFromCP!(
    execute
    self
    cp:Object Reference=@n10
    OverlayMaterials:Map of Strings to Material Interface Object References=Get OverlayMaterials() -> (OverlayMaterials:Map of Strings to Material Interface Object References)
   ) -> (then) {is=title K2Node_CallFunction}
   Add(
    execute=@n4.else
    TargetArray:Array of Custom Character Part Object References=Get CPs() -> (CPs:Array of Custom Character Part Object References)
    NewItem:"Custom Character Part Object Reference (by ref)"=Get CharacterPart()
   ) {is=title K2Node_CallArrayFunction}
   > n11 = Set CurrSkelMesh(CurrSkelMesh:Skeletal Mesh Object Reference=GetSkeletalMesh!(self:Custom Character Part Object Reference=Get CharacterPart()) -> (ReturnValue:Skeletal Mesh Object Reference)) -> (Output_Get:Skeletal Mesh Object Reference)
   > n12 = Add(TargetArray:Array of Skeletal Mesh Object References=Get Meshes(), NewItem:"Skeletal Mesh Object Reference (by ref)"=@n11)
  ]}
 LoadVariants {nodes [
   LoadVariants!() -> (then, Mannequin:Fort Player Mannequin Object Reference) {is=title K2Node_FunctionEntry}
   > Clear(TargetArray:Array of Material Interface Object References=Get ToSwapMaterials() -> (ToSwapMaterials:Array of Material Interface Object References)) {is=title K2Node_CallArrayFunction}
   > n1 = SelectCosmeticObject(Character=Get CharacterID(), Pickaxe=Get PickaxeID(), Backpack=Get BackpackID(), Glider=Get GliderID())
   > n2 = Set CurrID(CurrID:Object Reference=@n1) -> (Output_Get:Object Reference)
   > n3 = Execute Python Script!(
    execute
    PythonScript="import TechArt.CosmeticPicker.CosmeticPickerModule as CosmeticPicker\r\nnum = CosmeticPicker.getNumVariants(id)"
    id:Object Reference=@n2
   ) -> (then, ReturnValue, num:int) {inputs [id], outputs [num]}
   > n4 = For Loop(LastIndex=int - int(A=@n3.num))
   > n5 = Set CurrVariantIndex(CurrVariantIndex:int=@n4) -> (Output_Get:int=0)
   > n6 = Execute Python Script!(
    execute
    PythonScript="import TechArt.CosmeticPicker.CosmeticPickerModule as CosmeticPicker\r\n\r\ntype = CosmeticPicker.getVariantType(id, variant_index)"
    variant_index:int=@n5
    id=Get CurrID() -> (CurrID:Object Reference)
   ) -> (then, ReturnValue, type:String) {inputs [variant_index, id], outputs [type]}
   n7 = "Equal Exactly (String)"(A=@n6.type, B=FortCosmeticVariantApplyWrap)
   n8 = "Equal Exactly (String)"(A=@n6.type, B=FortCosmeticFloatSliderVariant)
   > n9 = Branch(Condition=OR Boolean(A=OR Boolean(A=@n7, B=@n8), B="Equal Exactly (String)"(A=@n6.type, B=FortCosmeticItemTexture)))
   n10 = Execute Python Script!(
    execute=@n9.else
    PythonScript="import TechArt.CosmeticPicker.CosmeticPickerModule as CosmeticPicker\r\n\r\ndefault_index = CosmeticPicker.getDefaultVariantDefaultOption(id, variant_index)\r\nactive_index = CosmeticPicker.checkMannequinForComboboxVariant(mannequin, id, variant_index, default_index)"
    mannequin:Fort Player Mannequin Object Reference=Get Mannequin() -> (Mannequin:Fort Player Mannequin Object Reference)
    id=Get CurrID()
    variant_index=Get CurrVariantIndex() -> (CurrVariantIndex:int=0)
   ) -> (then, ReturnValue, default_index:int, active_index:int) {inputs [mannequin, id, variant_index], outputs [default_index, active_index]}
   > Set DefaultIndex(DefaultIndex:int=@n10.default_index) -> (Output_Get:int=0)
   > n11 = Execute Python Script!(
    execute
    PythonScript="import TechArt.CosmeticPicker.CosmeticPickerModule as CosmeticPicker\r\n\r\nto_swap_materials = []\r\noverride_materials = []\r\n\r\nitemVariants = id.get_editor_property(\"itemVariants\")\r\nvariant_options = CosmeticPicker.getOptionsParam(itemVariants[variant_index])\r\nif active_index < len(variant_options):\r\n    variant_option = variant_options[active_index]\r\n    variant_materials = variant_option.get_editor_property(\"variantMaterials\")\r\n    \r\n    for m in variant_materials:\r\n        to_swap_materials.append(m.get_editor_property(\"MaterialtoSwap\"))\r\n        override_materials.append(m.get_editor_property(\"OverrideMaterial\"))"
    id=Get CurrID()
    variant_index=Get CurrVariantIndex() -> (CurrVariantIndex=0)
    active_index:int=@n10.active_index
   ) -> (
    then
    ReturnValue
    to_swap_materials:Array of Material Interface Object References
    override_materials:Array of Material Interface Object References
   ) {inputs [id, variant_index, active_index], outputs [to_swap_materials, override_materials]}
   > Branch(Condition="integer > integer"(A=Length(TargetArray:Array of Material Interface Object References=@n11.to_swap_materials)))
   > Set ToSwapMaterials(ToSwapMaterials:Array of Material Interface Object References=@n11.to_swap_materials) -> (Output_Get:Array of Material Interface Object References)
   > Set OverrideMaterials(OverrideMaterials:Array of Material Interface Object References=@n11.override_materials) -> (Output_Get:Array of Material Interface Object References)
   > SwapMaterials()
   n12 = Get OverrideMaterials() -> (OverrideMaterials:Array of Material Interface Object References)
   n13 = For Loop(LastIndex=int - int(A=Length(TargetArray=@n12)))
   n14 = "Get (a copy)"(Array:Array of Material Interface Object References=Get OverrideMaterials(), Dimension 1=@n13) -> (Output:Material Interface Object Reference)
   n15 = "Get (a copy)"(Array=Get ToSwapMaterials(), Dimension 1=@n13)
   > PrintString(InString=Append!(A=GetDisplayName(Object=@n15), B=" is swapping to ", C:String=GetDisplayName(Object=@n14)) -> (ReturnValue))
   Get CharacterID()
   Get CharacterID()
   Get ToSwapMaterials()
  ]}
 OnTabClicked {nodes [
   n1 = OnTabClicked!() -> (then, TabIndex:int) {is=title K2Node_FunctionEntry}
   > UpdateTabs!(execute, self, Tab:int=@n1) -> (then)
   > SetActiveWidgetIndex(self=Get WidgetSwitcher_Details() -> (WidgetSwitcher_Details:Widget Switcher Object Reference), Index=@n1)
   > Set TabIndex(TabIndex:int=@n1) -> (Output_Get:int=0)
  ]}
 SelectCosmeticObject {nodes [
   n1 = SelectCosmeticObject!() -> (then, Character:Object Reference, Pickaxe:Object Reference, Backpack:Object Reference, Glider:Object Reference) {is=title K2Node_FunctionEntry}
   > n2 = Branch(Condition="Equal (Integer)"(A=Get CurrAssetType() -> (CurrAssetType=0)))
   > Return Node!(execute, Object:Object Reference=@n1)
   n3 = Branch(execute=@n2.else, Condition="Equal (Integer)"(A=Get CurrAssetType() -> (CurrAssetType=0), B=1))
   > Return Node(Object=@n1.Pickaxe)
   n4 = Branch(execute=@n3.else, Condition="Equal (Integer)"(A=Get CurrAssetType() -> (CurrAssetType=0), B=2))
   > Return Node(Object=@n1.Backpack)
   Return Node(execute=@n4.else, Object=@n1.Glider)
  ]}
 ShowDetailsView {nodes [
   n1 = ShowDetailsView!() -> (then, DetailsView:Details View Object Reference, GroupName:Name) {is=title K2Node_FunctionEntry}
   > SetEditorProperty(Object=@n1, PropertyName=CategoriesToShow, PropertyValue:Array of Names=Make Array!("[0]":Name=@n1.GroupName) -> (Array:Array of Names))
   > SetObject(self=@n1, NewObject=Self-Reference())
   > SetVisibility(self=@n1, InVisibility=Visible)
  ]}
 SwapMaterials {nodes [
   SwapMaterials() {is=title K2Node_FunctionEntry}
   > n1 = Branch(Condition=Get isCharacter() -> (isCharacter:bool))
   > n2 = For Each Loop(Array=Get ToSwapMaterials())
   > Set OverrideMaterialIndex(OverrideMaterialIndex:int=@n2.Array Index) -> (Output_Get:int=0)
   n3 = Find Item!(TargetArray:Array of Material Interface Object References=Get BodyMaterials(), ItemToFind:"Material Interface Object Reference (by ref)"=@n2) -> (ReturnValue)
   > n4 = Branch(Condition="integer >= integer"(A=@n3))
   > Set Array Elem(
    TargetArray=Get BodyMaterials()
    Index=@n3
    Item="Get (a copy)"(Array=Get OverrideMaterials(), Dimension 1=Get OverrideMaterialIndex() -> (OverrideMaterialIndex:int=0))
    bSizeToFit=false
   )
   n5 = Find Item(TargetArray=Get HeadMaterials(), ItemToFind=@n2)
   n6 = Branch(execute=@n4.else, Condition="integer >= integer"(A=@n5))
   > Set Array Elem(
    TargetArray=Get HeadMaterials()
    Index=@n5
    Item="Get (a copy)"(Array=Get OverrideMaterials(), Dimension 1=Get OverrideMaterialIndex() -> (OverrideMaterialIndex=0))
    bSizeToFit=false
   )
   n7 = Find Item(TargetArray=Get FaceAccMaterials(), ItemToFind=@n2)
   Branch(execute=@n6.else, Condition="integer >= integer"(A=@n7))
   > Set Array Elem(
    TargetArray=Get FaceAccMaterials()
    Index=@n7
    Item="Get (a copy)"(Array=Get OverrideMaterials(), Dimension 1=Get OverrideMaterialIndex() -> (OverrideMaterialIndex=0))
    bSizeToFit=false
   )
   n8 = For Each Loop(Exec=@n1.else, Array=Get ToSwapMaterials())
   > Set OverrideMaterialIndex(OverrideMaterialIndex=0) -> (Output_Get=0)
   n9 = Find Item(TargetArray=Get Materials(), ItemToFind=@n8)
   > Branch(Condition="integer >= integer"(A=@n9))
   > Set Array Elem(
    TargetArray=Get Materials()
    Index=@n9
    Item="Get (a copy)"(Array=Get OverrideMaterials(), Dimension 1=Get OverrideMaterialIndex() -> (OverrideMaterialIndex=0))
    bSizeToFit=false
   )
  ]}
 UpdateCheckboxes {nodes [
   n1 = UpdateCheckboxes!() -> (then, Checked:Editor Utility Check Box Object Reference) {is=title K2Node_FunctionEntry}
   > n2 = For Each Loop(Array:Array of Editor Utility Check Box Object References=Get Checkboxes()) -> (Array Element:Editor Utility Check Box Object Reference)
   SetCheckedState!(execute=@n2.Completed, self:Check Box Object Reference=@n1, InCheckedState:ECheckBoxState Enum=Checked) -> (then)
   SetCheckedState(execute=@n2.LoopBody, self=@n2, InCheckedState=Unchecked)
  ]}
 UpdateTabs {nodes [
   n1 = UpdateTabs!() -> (then, Tab:int) {is=title K2Node_FunctionEntry}
   > n2 = Branch(Condition=Is Empty!(TargetArray:Array of Editor Utility Button Object References=Get Buttons()) -> (ReturnValue:bool))
   > PrintString(InString=ERROR: Buttons array is empty, TextColor=(R=1,G=0.0233,B=0,A=1))
   n3 = For Each Loop(Exec=@n2.else, Array:Array of Editor Utility Button Object References=Get Buttons()) -> (Array Element:Editor Utility Button Object Reference)
   > n4 = Branch(Condition="Equal (Integer)"(A=@n3.Array Index, B=@n1))
   > SetBackgroundColor!(execute, self:Button Object Reference=@n3, InBackgroundColor:Linear Color Structure=(R=2,G=2,B=2,A=1)) -> (then)
   SetBackgroundColor(execute=@n4.else, self=@n3, InBackgroundColor=(R=1,G=1,B=1,A=1))
  ]}
 UpdateUI {nodes [
   UpdateUI() {is=title K2Node_FunctionEntry}
   n1 = Get CurrCosmeticType() -> (CurrCosmeticType=NewEnumerator0)
   > ShowDetailsView!(
    execute
    self
    DetailsView:Details View Object Reference=Select!(
     NewEnumerator0:Details View Object Reference=Get DetailsView_CID() -> (DetailsView_CID:Details View Object Reference)
     NewEnumerator1:Details View Object Reference=Get DetailsView_BID() -> (DetailsView_BID:Details View Object Reference)
     NewEnumerator2:Details View Object Reference=Get DetailsView_PID() -> (DetailsView_PID:Details View Object Reference)
     NewEnumerator3:Details View Object Reference=Get DetailsView_GID() -> (DetailsView_GID:Details View Object Reference)
     NewEnumerator4:Details View Object Reference
     NewEnumerator5:Details View Object Reference
     NewEnumerator6:Details View Object Reference
     NewEnumerator7:Details View Object Reference
     NewEnumerator8:Details View Object Reference
     NewEnumerator9:Details View Object Reference
     NewEnumerator10:Details View Object Reference
     NewEnumerator11:Details View Object Reference
     NewEnumerator12:Details View Object Reference
     NewEnumerator13:Details View Object Reference
     NewEnumerator14:Details View Object Reference
     NewEnumerator15:Details View Object Reference
     Index:CosmeticPicker_CosmeticTypes Enum=@n1
    ) -> (ReturnValue:Details View Object Reference)
    GroupName:Name=Select(
     NewEnumerator0:Name=CID
     NewEnumerator1:Name=BID
     NewEnumerator2:Name=PID
     NewEnumerator3:Name=GID
     NewEnumerator4:Name=None
     NewEnumerator5:Name=None
     NewEnumerator6:Name=None
     NewEnumerator7:Name=None
     NewEnumerator8:Name=None
     NewEnumerator9:Name=None
     NewEnumerator10:Name=None
     NewEnumerator11:Name=None
     NewEnumerator12:Name=None
     NewEnumerator13:Name=None
     NewEnumerator14:Name=None
     NewEnumerator15:Name=None
     Index=@n1
    ) -> (ReturnValue:Name)
   ) -> (then) {is=title K2Node_CallFunction}
   > SetActiveWidgetIndex(self=Get WidgetSwitcher_ID() -> (WidgetSwitcher_ID:Widget Switcher Object Reference), Index=Get CurrAssetType() -> (CurrAssetType=0))
   > n2 = Branch(Condition=Get isCharacter())
   > ShowDetailsView(
    DetailsView=Select!(
     Option 0:Details View Object Reference=Get DetailsView_Materials() -> (DetailsView_Materials:Details View Object Reference)
     Option 1:Details View Object Reference=Get DetailsView_Meshes() -> (DetailsView_Meshes:Details View Object Reference)
     Option 2:Details View Object Reference=Get DetailsView_CPs() -> (DetailsView_CPs:Details View Object Reference)
     Index:int=Get TabIndex() -> (TabIndex:int=0)
    ) -> (ReturnValue:Details View Object Reference)
    GroupName=Select(
     Option 0:Name=String To Name(InString=Materials)
     Option 1:Name=String To Name(InString=Meshes)
     Option 2:Name=String To Name(InString=CPs)
     Index=Get TabIndex() -> (TabIndex=0)
    ) -> (ReturnValue:Name)
   )
   > HideDetailsView!(
    execute
    self
    Details View:Details View Object Reference=Select(
     Option 0:Details View Object Reference=Get DetailsView_AccMaterials() -> (DetailsView_AccMaterials:Details View Object Reference)
     Option 1:Details View Object Reference=Get DetailsView_AccMeshes() -> (DetailsView_AccMeshes:Details View Object Reference)
     Option 2:Details View Object Reference
     Index=Get TabIndex() -> (TabIndex=0)
    ) -> (ReturnValue:Details View Object Reference)
   ) -> (then) {is=title K2Node_CallFunction}
   ShowDetailsView(
    execute=@n2.else
    DetailsView=Select(Option 0=Get DetailsView_AccMaterials(), Option 1=Get DetailsView_AccMeshes(), Index=Get TabIndex() -> (TabIndex=0))
    GroupName=Select(
     Option 0:Name=String To Name(InString=AccMaterials)
     Option 1:Name=String To Name(InString=AccMeshes)
     Option 2:Name=String To Name(InString=CPs)
     Index=Get TabIndex() -> (TabIndex=0)
    ) -> (ReturnValue:Name)
   )
   > HideDetailsView(Details View=Select(
     Option 0:Details View Object Reference=Get DetailsView_Materials()
     Option 1:Details View Object Reference=Get DetailsView_Meshes()
     Option 2:Details View Object Reference=Get DetailsView_CPs()
     Index=Get TabIndex() -> (TabIndex=0)
    ) -> (ReturnValue:Details View Object Reference))
  ]}
]
variables [
 AccOverlayMaterials {type="map<MaterialInterface>", default {}}
 BackpackID {type=AthenaBackpackItemDefinition, default=null}
 BodyCPs {type="array<CustomCharacterPart>", default []}
 BodyMaterials {type="array<MaterialInterface>", default []}
 BodyOverlayMaterials {type="map<MaterialInterface>", default {}}
 BodySkelMeshes {type="array<SkeletalMesh>", default []}
 Buttons {type="array<EditorUtilityButton>", default []}
 CharacterID {type=AthenaCharacterItemDefinition, default=null}
 CharacterParts {type="array<CustomCharacterPart>", default []}
 Checkboxes {type="array<EditorUtilityCheckBox>", default []}
 Curr Asset Type {type=integer}
 CurrAssetType {type=integer}
 CurrCosmeticType {type=CosmeticPicker_CosmeticTypes, default=Character}
 CurrSelection Set {type=TypedElementSelectionSet, default=null}
 FaceAccMaterials {type="array<MaterialInterface>", default []}
 FaceAccOverlayMaterials {type="map<MaterialInterface>", default {}}
 FaceCPs {type="array<CustomCharacterPart>", default []}
 FaceSkelMeshes {type="array<SkeletalMesh>", default []}
 GliderID {type=AthenaGliderItemDefinition, default=null}
 HatCPs {type="array<CustomCharacterPart>", default []}
 HatMaterials {type="array<MaterialInterface>", default []}
 HatOverlayMaterials {type="map<MaterialInterface>", default {}}
 HatSkelMeshes {type="array<SkeletalMesh>", default []}
 HeadCPs {type="array<CustomCharacterPart>", default []}
 HeadMaterials {type="array<MaterialInterface>", default []}
 HeadOverlayMaterials {type="map<MaterialInterface>", default {}}
 HeadSkelMeshes {type="array<SkeletalMesh>", default []}
 isCharacter {type=boolean, default=true}
 Materials {type="array<MaterialInterface>", default []}
 Meshes {type="array<SkeletalMesh>", default []}
 MiscTailCPs {type="array<CustomCharacterPart>", default []}
 MiscTailMaterials {type="array<MaterialInterface>", default []}
 MiscTailOverlayMaterials {type="map<MaterialInterface>", default {}}
 MiscTailSkelMeshes {type="array<SkeletalMesh>", default []}
 OverrideMaterials {type="array<MaterialInterface>", default []}
 PickaxeID {type=AthenaPickaxeItemDefinition, default=null}
 Property_OverrideVariants {type="array<McpVariantChannelInfo>", default []}
 ShowOverlayMaterials {type=boolean}
 TabIndex {type=integer}
 ThisMannequin {type=FortPlayerMannequin, default=null}
 ToSwapMaterials {type="array<MaterialInterface>", default []}
 VariantTags {type="array<GameplayTag>", default []}
]
widgets {
 info {rootWidgetClass=&StackBox, widgetCount=47}
 widgets [
  StackBox_0 {widgetClassPath=&StackBox, widget=StackBox_0}
  VerticalBox_78 {
   widgetClassPath=&VerticalBox
   parent=StackBox_0
   widget=VerticalBox_78
   slot {class=StackBoxSlot, properties {padding=<ltrb 5 5 5 5>, size {sizeRule=Fill, value=1}}}
  }
  HorizontalBox_81 {
   widgetClassPath=&HorizontalBox
   parent=VerticalBox_78
   widget=HorizontalBox_81
   slot {class=VerticalBoxSlot, properties {horizontalAlignment=HAlign_Left}}
  }
  Btn_Refresh {
   widgetClassPath=&EditorUtilityButton
   parent=HorizontalBox_81
   widget=Btn_Refresh
   properties {toolTipText=Reload asset if detail view is not updated.}
   slot {class=HorizontalBoxSlot, properties {horizontalAlignment=HAlign_Right}}
  }
  TextBlock_47 {
   widgetClassPath=&TextBlock
   parent=Btn_Refresh
   widget=TextBlock_47
   properties {
    text=Refresh
    font {
     bForceMonospaced=false
     bMaterialIsStencil=false
     fontMaterial=null
     fontObject=&Roboto
     letterSpacing=0
     monospacedWidth=1
     size=9
     skewAmount=0
     typefaceFontName=Regular
     outlineSettings {
      bApplyOutlineToDropShadows=false
      bMiteredCorners=false
      bSeparateFillAlpha=false
      outlineMaterial=null
      outlineSize=0
      outlineColor=<rgba 0 0 0 1>
     }
    }
   }
   slot {class=ButtonSlot}
  }
  CheckBox_ShowOverlayMat {
   widgetClassPath=&EditorUtilityCheckBox
   bIsVariable=true
   parent=HorizontalBox_81
   widget=CheckBox_ShowOverlayMat
   properties {visibility=Collapsed}
   slot {class=HorizontalBoxSlot, properties {padding=<ltrb 30 0 0 0>}}
  }
  TextBlock_3 {
   widgetClassPath=&TextBlock
   parent=CheckBox_ShowOverlayMat
   widget=TextBlock_3
   properties {
    text=Show OverlayMaterial
    font {
     bForceMonospaced=false
     bMaterialIsStencil=false
     fontMaterial=null
     fontObject=&Roboto
     letterSpacing=0
     monospacedWidth=1
     size=9
     skewAmount=0
     typefaceFontName=Regular
     outlineSettings {
      bApplyOutlineToDropShadows=false
      bMiteredCorners=false
      bSeparateFillAlpha=false
      outlineMaterial=null
      outlineSize=0
      outlineColor=<rgba 0 0 0 1>
     }
    }
   }
   slot {class=PanelSlot}
  }
  HB_AssetType {
   widgetClassPath=&HorizontalBox
   parent=VerticalBox_78
   widget=HB_AssetType
   slot {class=VerticalBoxSlot, properties {padding=<ltrb 0 10 0 10>}}
  }
  Checkbox_Character {
   widgetClassPath=&EditorUtilityCheckBox
   bIsVariable=true
   parent=HB_AssetType
   widget=Checkbox_Character
   properties {checkedState=Checked}
   slot {class=HorizontalBoxSlot, properties {padding=<ltrb 10 0 10 0>}}
  }
  TextBlock_61 {
   widgetClassPath=&TextBlock
   parent=Checkbox_Character
   widget=TextBlock_61
   properties {
    text=Character
    font {
     bForceMonospaced=false
     bMaterialIsStencil=false
     fontMaterial=null
     fontObject=&Roboto
     letterSpacing=0
     monospacedWidth=1
     size=9
     skewAmount=0
     typefaceFontName=Regular
     outlineSettings {
      bApplyOutlineToDropShadows=false
      bMiteredCorners=false
      bSeparateFillAlpha=false
      outlineMaterial=null
      outlineSize=0
      outlineColor=<rgba 0 0 0 1>
     }
    }
   }
   slot {class=PanelSlot}
  }
  Checkbox_Pickaxe {
   widgetClassPath=&EditorUtilityCheckBox
   bIsVariable=true
   parent=HB_AssetType
   widget=Checkbox_Pickaxe
   slot {class=HorizontalBoxSlot, properties {padding=<ltrb 10 0 10 0>}}
  }
  TextBlock_4 {
   widgetClassPath=&TextBlock
   parent=Checkbox_Pickaxe
   widget=TextBlock_4
   properties {
    text=Pickaxe
    font {
     bForceMonospaced=false
     bMaterialIsStencil=false
     fontMaterial=null
     fontObject=&Roboto
     letterSpacing=0
     monospacedWidth=1
     size=9
     skewAmount=0
     typefaceFontName=Regular
     outlineSettings {
      bApplyOutlineToDropShadows=false
      bMiteredCorners=false
      bSeparateFillAlpha=false
      outlineMaterial=null
      outlineSize=0
      outlineColor=<rgba 0 0 0 1>
     }
    }
   }
   slot {class=PanelSlot}
  }
  Checkbox_Backpack {
   widgetClassPath=&EditorUtilityCheckBox
   bIsVariable=true
   parent=HB_AssetType
   widget=Checkbox_Backpack
   slot {class=HorizontalBoxSlot, properties {padding=<ltrb 10 0 10 0>}}
  }
  TextBlock_5 {
   widgetClassPath=&TextBlock
   parent=Checkbox_Backpack
   widget=TextBlock_5
   properties {
    text=Backpack
    font {
     bForceMonospaced=false
     bMaterialIsStencil=false
     fontMaterial=null
     fontObject=&Roboto
     letterSpacing=0
     monospacedWidth=1
     size=9
     skewAmount=0
     typefaceFontName=Regular
     outlineSettings {
      bApplyOutlineToDropShadows=false
      bMiteredCorners=false
      bSeparateFillAlpha=false
      outlineMaterial=null
      outlineSize=0
      outlineColor=<rgba 0 0 0 1>
     }
    }
   }
   slot {class=PanelSlot}
  }
  Checkbox_Glider {
   widgetClassPath=&EditorUtilityCheckBox
   bIsVariable=true
   parent=HB_AssetType
   widget=Checkbox_Glider
   slot {class=HorizontalBoxSlot, properties {padding=<ltrb 10 0 10 0>}}
  }
  TextBlock_6 {
   widgetClassPath=&TextBlock
   parent=Checkbox_Glider
   widget=TextBlock_6
   properties {
    text=Glider
    font {
     bForceMonospaced=false
     bMaterialIsStencil=false
     fontMaterial=null
     fontObject=&Roboto
     letterSpacing=0
     monospacedWidth=1
     size=9
     skewAmount=0
     typefaceFontName=Regular
     outlineSettings {
      bApplyOutlineToDropShadows=false
      bMiteredCorners=false
      bSeparateFillAlpha=false
      outlineMaterial=null
      outlineSize=0
      outlineColor=<rgba 0 0 0 1>
     }
    }
   }
   slot {class=PanelSlot}
  }
  WidgetSwitcher_ID {widgetClassPath=&WidgetSwitcher, bIsVariable=true, parent=VerticalBox_78, widget=WidgetSwitcher_ID, slot {class=VerticalBoxSlot}}
  Border_85 {
   widgetClassPath=&Border
   parent=WidgetSwitcher_ID
   widget=Border_85
   properties {brushColor=<rgba 0.0395 0.0395 0.0395 1>, padding=<ltrb 8 8 8 8>}
   slot {class=WidgetSwitcherSlot}
  }
  DetailsView_CID {
   widgetClassPath=&DetailsView
   parent=Border_85
   widget=DetailsView_CID
   slot {class=BorderSlot, properties {padding=<ltrb 8 8 8 8>}}
  }
  Border {
   widgetClassPath=&Border
   parent=WidgetSwitcher_ID
   widget=Border
   properties {brushColor=<rgba 0.0395 0.0395 0.0395 1>, padding=<ltrb 8 8 8 8>}
   slot {class=WidgetSwitcherSlot}
  }
  DetailsView_PID {
   widgetClassPath=&DetailsView
   parent=Border
   widget=DetailsView_PID
   slot {class=BorderSlot, properties {padding=<ltrb 8 8 8 8>}}
  }
  Border_1 {
   widgetClassPath=&Border
   parent=WidgetSwitcher_ID
   widget=Border_1
   properties {brushColor=<rgba 0.0395 0.0395 0.0395 1>, padding=<ltrb 8 8 8 8>}
   slot {class=WidgetSwitcherSlot}
  }
  DetailsView_BID {
   widgetClassPath=&DetailsView
   parent=Border_1
   widget=DetailsView_BID
   slot {class=BorderSlot, properties {padding=<ltrb 8 8 8 8>}}
  }
  Border_2 {
   widgetClassPath=&Border
   parent=WidgetSwitcher_ID
   widget=Border_2
   properties {brushColor=<rgba 0.0395 0.0395 0.0395 1>, padding=<ltrb 8 8 8 8>}
   slot {class=WidgetSwitcherSlot}
  }
  DetailsView_GID {
   widgetClassPath=&DetailsView
   parent=Border_2
   widget=DetailsView_GID
   slot {class=BorderSlot, properties {padding=<ltrb 8 8 8 8>}}
  }
  Spacer_88 {widgetClassPath=&Spacer, parent=VerticalBox_78, widget=Spacer_88, properties {size=<1 10>}, slot {class=VerticalBoxSlot}}
  HorizontalBox_49 {widgetClassPath=&HorizontalBox, parent=VerticalBox_78, widget=HorizontalBox_49, slot {class=VerticalBoxSlot}}
  Btn_Material {
   widgetClassPath=&EditorUtilityButton
   parent=HorizontalBox_49
   widget=Btn_Material
   properties {widgetStyle {
     clickedSlateSound {resourceObject=null}
     disabled {
      bIsDynamicallyLoaded=false
      drawAs=Box
      imageType=NoImage
      mirroring=NoMirror
      resourceName=None
      resourceObject=null
      tiling=NoTile
      imageSize=<0 0>
      margin=<ltrb 0 0 0 0>
      outlineSettings {
       bUseBrushTransparency=false
       roundingType=FixedRadius
       width=1
       color {colorUseRule=UseColor_ColorTable, specifiedColor=<rgba 0 0 0 0>}
       cornerRadii=<4 4 4 4>
      }
      tintColor {colorUseRule=UseColor_Specified, specifiedColor=<rgba 0.0395 0.0395 0.0395 1>}
      uVRegion {bIsValid=false, max=<0 0>, min=<0 0>}
     }
     disabledForeground {colorUseRule=UseColor_ColorTable, specifiedColor=<rgba 0 0 0 0>}
     hovered {
      bIsDynamicallyLoaded=false
      drawAs=Box
      imageType=NoImage
      mirroring=NoMirror
      resourceName=None
      resourceObject=null
      tiling=NoTile
      imageSize=<32 32>
      margin=<ltrb 0 0 0 0>
      outlineSettings {
       bUseBrushTransparency=false
       roundingType=FixedRadius
       width=1
       color {colorUseRule=UseColor_ColorTable, specifiedColor=<rgba 0 0 0 0>}
       cornerRadii=<4 4 4 4>
      }
      tintColor {colorUseRule=UseColor_Specified, specifiedColor=<rgba 0.0953 0.0953 0.0953 1>}
      uVRegion {bIsValid=false, max=<0 0>, min=<0 0>}
     }
     hoveredForeground {colorUseRule=UseColor_ColorTable, specifiedColor=<rgba 0 0 0 0>}
     hoveredSlateSound {resourceObject=null}
     normal {
      bIsDynamicallyLoaded=false
      drawAs=Box
      imageType=NoImage
      mirroring=NoMirror
      resourceName=None
      resourceObject=null
      tiling=NoTile
      imageSize=<32 32>
      margin=<ltrb 0 0 0 0>
      outlineSettings {
       bUseBrushTransparency=false
       roundingType=FixedRadius
       width=1
       color {colorUseRule=UseColor_ColorTable, specifiedColor=<rgba 0 0 0 0>}
       cornerRadii=<4 4 4 4>
      }
      tintColor {colorUseRule=UseColor_Specified, specifiedColor=<rgba 0.0395 0.0395 0.0395 1>}
      uVRegion {bIsValid=false, max=<0 0>, min=<0 0>}
     }
     normalForeground {colorUseRule=UseColor_ColorTable, specifiedColor=<rgba 0 0 0 0>}
     normalPadding=<ltrb 12 1.5 12 1.5>
     pressed {
      bIsDynamicallyLoaded=false
      drawAs=Box
      imageType=NoImage
      mirroring=NoMirror
      resourceName=None
      resourceObject=null
      tiling=NoTile
      imageSize=<32 32>
      margin=<ltrb 0 0 0 0>
      outlineSettings {
       bUseBrushTransparency=false
       roundingType=FixedRadius
       width=1
       color {colorUseRule=UseColor_ColorTable, specifiedColor=<rgba 0 0 0 0>}
       cornerRadii=<4 4 4 4>
      }
      tintColor {colorUseRule=UseColor_Specified, specifiedColor=<rgba 0.0284 0.0284 0.0284 1>}
      uVRegion {bIsValid=false, max=<0 0>, min=<0 0>}
     }
     pressedForeground {colorUseRule=UseColor_ColorTable, specifiedColor=<rgba 0 0 0 0>}
     pressedPadding=<ltrb 12 2.5 12 0.5>
     pressedSlateSound {resourceObject=null}
    }}
   slot {class=HorizontalBoxSlot, properties {padding=<ltrb 0 0 5 0>}}
  }
  TextBlock_64 {
   widgetClassPath=&TextBlock
   parent=Btn_Material
   widget=TextBlock_64
   properties {
    text=Material
    font {
     bForceMonospaced=false
     bMaterialIsStencil=false
     fontMaterial=null
     fontObject=&Roboto
     letterSpacing=0
     monospacedWidth=1
     size=10
     skewAmount=0
     typefaceFontName=Regular
     outlineSettings {
      bApplyOutlineToDropShadows=false
      bMiteredCorners=false
      bSeparateFillAlpha=false
      outlineMaterial=null
      outlineSize=0
      outlineColor=<rgba 0 0 0 1>
     }
    }
   }
   slot {class=ButtonSlot}
  }
  Btn_SkelMesh {
   widgetClassPath=&EditorUtilityButton
   parent=HorizontalBox_49
   widget=Btn_SkelMesh
   properties {widgetStyle {
     clickedSlateSound {resourceObject=null}
     disabled {
      bIsDynamicallyLoaded=false
      drawAs=Box
      imageType=NoImage
      mirroring=NoMirror
      resourceName=None
      resourceObject=null
      tiling=NoTile
      imageSize=<0 0>
      margin=<ltrb 0 0 0 0>
      outlineSettings {
       bUseBrushTransparency=false
       roundingType=FixedRadius
       width=1
       color {colorUseRule=UseColor_ColorTable, specifiedColor=<rgba 0 0 0 0>}
       cornerRadii=<4 4 4 4>
      }
      tintColor {colorUseRule=UseColor_Specified, specifiedColor=<rgba 0.0395 0.0395 0.0395 1>}
      uVRegion {bIsValid=false, max=<0 0>, min=<0 0>}
     }
     disabledForeground {colorUseRule=UseColor_ColorTable, specifiedColor=<rgba 0 0 0 0>}
     hovered {
      bIsDynamicallyLoaded=false
      drawAs=Box
      imageType=NoImage
      mirroring=NoMirror
      resourceName=None
      resourceObject=null
      tiling=NoTile
      imageSize=<32 32>
      margin=<ltrb 0 0 0 0>
      outlineSettings {
       bUseBrushTransparency=false
       roundingType=FixedRadius
       width=1
       color {colorUseRule=UseColor_ColorTable, specifiedColor=<rgba 0 0 0 0>}
       cornerRadii=<4 4 4 4>
      }
      tintColor {colorUseRule=UseColor_Specified, specifiedColor=<rgba 0.0953 0.0953 0.0953 1>}
      uVRegion {bIsValid=false, max=<0 0>, min=<0 0>}
     }
     hoveredForeground {colorUseRule=UseColor_ColorTable, specifiedColor=<rgba 0 0 0 0>}
     hoveredSlateSound {resourceObject=null}
     normal {
      bIsDynamicallyLoaded=false
      drawAs=Box
      imageType=NoImage
      mirroring=NoMirror
      resourceName=None
      resourceObject=null
      tiling=NoTile
      imageSize=<32 32>
      margin=<ltrb 0 0 0 0>
      outlineSettings {
       bUseBrushTransparency=false
       roundingType=FixedRadius
       width=1
       color {colorUseRule=UseColor_ColorTable, specifiedColor=<rgba 0 0 0 0>}
       cornerRadii=<4 4 4 4>
      }
      tintColor {colorUseRule=UseColor_Specified, specifiedColor=<rgba 0.0395 0.0395 0.0395 1>}
      uVRegion {bIsValid=false, max=<0 0>, min=<0 0>}
     }
     normalForeground {colorUseRule=UseColor_ColorTable, specifiedColor=<rgba 0 0 0 0>}
     normalPadding=<ltrb 12 1.5 12 1.5>
     pressed {
      bIsDynamicallyLoaded=false
      drawAs=Box
      imageType=NoImage
      mirroring=NoMirror
      resourceName=None
      resourceObject=null
      tiling=NoTile
      imageSize=<32 32>
      margin=<ltrb 0 0 0 0>
      outlineSettings {
       bUseBrushTransparency=false
       roundingType=FixedRadius
       width=1
       color {colorUseRule=UseColor_ColorTable, specifiedColor=<rgba 0 0 0 0>}
       cornerRadii=<4 4 4 4>
      }
      tintColor {colorUseRule=UseColor_Specified, specifiedColor=<rgba 0.0284 0.0284 0.0284 1>}
      uVRegion {bIsValid=false, max=<0 0>, min=<0 0>}
     }
     pressedForeground {colorUseRule=UseColor_ColorTable, specifiedColor=<rgba 0 0 0 0>}
     pressedPadding=<ltrb 12 2.5 12 0.5>
     pressedSlateSound {resourceObject=null}
    }}
   slot {class=HorizontalBoxSlot, properties {padding=<ltrb 0 0 5 0>}}
  }
  TextBlock_1 {
   widgetClassPath=&TextBlock
   parent=Btn_SkelMesh
   widget=TextBlock_1
   properties {
    text=Skeletal Mesh
    font {
     bForceMonospaced=false
     bMaterialIsStencil=false
     fontMaterial=null
     fontObject=&Roboto
     letterSpacing=0
     monospacedWidth=1
     size=10
     skewAmount=0
     typefaceFontName=Regular
     outlineSettings {
      bApplyOutlineToDropShadows=false
      bMiteredCorners=false
      bSeparateFillAlpha=false
      outlineMaterial=null
      outlineSize=0
      outlineColor=<rgba 0 0 0 1>
     }
    }
   }
   slot {class=ButtonSlot}
  }
  Btn_CP {
   widgetClassPath=&EditorUtilityButton
   parent=HorizontalBox_49
   widget=Btn_CP
   properties {widgetStyle {
     clickedSlateSound {resourceObject=null}
     disabled {
      bIsDynamicallyLoaded=false
      drawAs=Box
      imageType=NoImage
      mirroring=NoMirror
      resourceName=None
      resourceObject=null
      tiling=NoTile
      imageSize=<0 0>
      margin=<ltrb 0 0 0 0>
      outlineSettings {
       bUseBrushTransparency=false
       roundingType=FixedRadius
       width=1
       color {colorUseRule=UseColor_ColorTable, specifiedColor=<rgba 0 0 0 0>}
       cornerRadii=<4 4 4 4>
      }
      tintColor {colorUseRule=UseColor_Specified, specifiedColor=<rgba 0.0395 0.0395 0.0395 1>}
      uVRegion {bIsValid=false, max=<0 0>, min=<0 0>}
     }
     disabledForeground {colorUseRule=UseColor_ColorTable, specifiedColor=<rgba 0 0 0 0>}
     hovered {
      bIsDynamicallyLoaded=false
      drawAs=Box
      imageType=NoImage
      mirroring=NoMirror
      resourceName=None
      resourceObject=null
      tiling=NoTile
      imageSize=<32 32>
      margin=<ltrb 0 0 0 0>
      outlineSettings {
       bUseBrushTransparency=false
       roundingType=FixedRadius
       width=1
       color {colorUseRule=UseColor_ColorTable, specifiedColor=<rgba 0 0 0 0>}
       cornerRadii=<4 4 4 4>
      }
      tintColor {colorUseRule=UseColor_Specified, specifiedColor=<rgba 0.0953 0.0953 0.0953 1>}
      uVRegion {bIsValid=false, max=<0 0>, min=<0 0>}
     }
     hoveredForeground {colorUseRule=UseColor_ColorTable, specifiedColor=<rgba 0 0 0 0>}
     hoveredSlateSound {resourceObject=null}
     normal {
      bIsDynamicallyLoaded=false
      drawAs=Box
      imageType=NoImage
      mirroring=NoMirror
      resourceName=None
      resourceObject=null
      tiling=NoTile
      imageSize=<32 32>
      margin=<ltrb 0 0 0 0>
      outlineSettings {
       bUseBrushTransparency=false
       roundingType=FixedRadius
       width=1
       color {colorUseRule=UseColor_ColorTable, specifiedColor=<rgba 0 0 0 0>}
       cornerRadii=<4 4 4 4>
      }
      tintColor {colorUseRule=UseColor_Specified, specifiedColor=<rgba 0.0395 0.0395 0.0395 1>}
      uVRegion {bIsValid=false, max=<0 0>, min=<0 0>}
     }
     normalForeground {colorUseRule=UseColor_ColorTable, specifiedColor=<rgba 0 0 0 0>}
     normalPadding=<ltrb 12 1.5 12 1.5>
     pressed {
      bIsDynamicallyLoaded=false
      drawAs=Box
      imageType=NoImage
      mirroring=NoMirror
      resourceName=None
      resourceObject=null
      tiling=NoTile
      imageSize=<32 32>
      margin=<ltrb 0 0 0 0>
      outlineSettings {
       bUseBrushTransparency=false
       roundingType=FixedRadius
       width=1
       color {colorUseRule=UseColor_ColorTable, specifiedColor=<rgba 0 0 0 0>}
       cornerRadii=<4 4 4 4>
      }
      tintColor {colorUseRule=UseColor_Specified, specifiedColor=<rgba 0.0284 0.0284 0.0284 1>}
      uVRegion {bIsValid=false, max=<0 0>, min=<0 0>}
     }
     pressedForeground {colorUseRule=UseColor_ColorTable, specifiedColor=<rgba 0 0 0 0>}
     pressedPadding=<ltrb 12 2.5 12 0.5>
     pressedSlateSound {resourceObject=null}
    }}
   slot {class=HorizontalBoxSlot, properties {padding=<ltrb 0 0 5 0>}}
  }
  TextBlock {
   widgetClassPath=&TextBlock
   parent=Btn_CP
   widget=TextBlock
   properties {
    text=Character Part
    font {
     bForceMonospaced=false
     bMaterialIsStencil=false
     fontMaterial=null
     fontObject=&Roboto
     letterSpacing=0
     monospacedWidth=1
     size=10
     skewAmount=0
     typefaceFontName=Regular
     outlineSettings {
      bApplyOutlineToDropShadows=false
      bMiteredCorners=false
      bSeparateFillAlpha=false
      outlineMaterial=null
      outlineSize=0
      outlineColor=<rgba 0 0 0 1>
     }
    }
   }
   slot {class=ButtonSlot}
  }
  EditorUtilityScrollBox_1 {
   widgetClassPath=&EditorUtilityScrollBox
   parent=VerticalBox_78
   widget=EditorUtilityScrollBox_1
   slot {class=VerticalBoxSlot, properties {size {sizeRule=Fill, value=1}}}
  }
  WidgetSwitcher_Details {
   widgetClassPath=&WidgetSwitcher
   bIsVariable=true
   parent=EditorUtilityScrollBox_1
   widget=WidgetSwitcher_Details
   slot {class=ScrollBoxSlot}
  }
  VerticalBox_68 {widgetClassPath=&VerticalBox, parent=WidgetSwitcher_Details, widget=VerticalBox_68, slot {class=WidgetSwitcherSlot}}
  Border_81 {
   widgetClassPath=&Border
   parent=VerticalBox_68
   widget=Border_81
   properties {brushColor=<rgba 0.0833 0.0833 0.0833 1>, padding=<ltrb 8 8 8 8>}
   slot {class=VerticalBoxSlot}
  }
  DetailsView_Materials {
   widgetClassPath=&DetailsView
   parent=Border_81
   widget=DetailsView_Materials
   slot {class=BorderSlot, properties {padding=<ltrb 8 8 8 8>}}
  }
  Border_3 {
   widgetClassPath=&Border
   parent=VerticalBox_68
   widget=Border_3
   properties {brushColor=<rgba 0.0833 0.0833 0.0833 1>, padding=<ltrb 8 8 8 8>}
   slot {class=VerticalBoxSlot}
  }
  DetailsView_AccMaterials {
   widgetClassPath=&DetailsView
   parent=Border_3
   widget=DetailsView_AccMaterials
   slot {class=BorderSlot, properties {padding=<ltrb 8 8 8 8>}}
  }
  VerticalBox_70 {widgetClassPath=&VerticalBox, parent=WidgetSwitcher_Details, widget=VerticalBox_70, slot {class=WidgetSwitcherSlot}}
  Border_83 {
   widgetClassPath=&Border
   parent=VerticalBox_70
   widget=Border_83
   properties {brushColor=<rgba 0.0833 0.0833 0.0833 1>, padding=<ltrb 8 8 8 8>}
   slot {class=VerticalBoxSlot}
  }
  DetailsView_Meshes {
   widgetClassPath=&DetailsView
   parent=Border_83
   widget=DetailsView_Meshes
   slot {class=BorderSlot, properties {padding=<ltrb 8 8 8 8>}}
  }
  Border_4 {
   widgetClassPath=&Border
   parent=VerticalBox_70
   widget=Border_4
   properties {brushColor=<rgba 0.0833 0.0833 0.0833 1>, padding=<ltrb 8 8 8 8>}
   slot {class=VerticalBoxSlot}
  }
  DetailsView_AccMeshes {
   widgetClassPath=&DetailsView
   parent=Border_4
   widget=DetailsView_AccMeshes
   slot {class=BorderSlot, properties {padding=<ltrb 8 8 8 8>}}
  }
  Border_84 {
   widgetClassPath=&Border
   parent=WidgetSwitcher_Details
   widget=Border_84
   properties {brushColor=<rgba 0.0833 0.0833 0.0833 1>, padding=<ltrb 8 8 8 8>}
   slot {class=WidgetSwitcherSlot}
  }
  DetailsView_CPs {
   widgetClassPath=&DetailsView
   parent=Border_84
   widget=DetailsView_CPs
   slot {class=BorderSlot, properties {padding=<ltrb 8 8 8 8>}}
  }
 ]
}
